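# Translucent unlit Niagara sprite material for chromatic fog particles, tuned per particle through Dynamic Parameters.
# Two panning turbulence samples, warped by counter-rotating smoke except on Low quality and mobile, set a sphere mask's radius to form ragged puffs.
# Optional chromatic tint packs three flow-distorted swirl noise samples into RGB with adjustable desaturation.
# Particle color and optional time-of-day values scale emissive and opacity, with depth fade skipped on Low quality.
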Material /FNGameplayCues/Athena/Effects/Materials/Generic/M_Chromatic_Fog_Particles
graph {
 properties {bUsedWithNiagaraSprites=true, blendMode=BLEND_Translucent, pixelDepthOffsetMode=PDOM_Legacy, shadingModel=MSM_Unlit, usedShadingModels=Unlit}
 nodes [
  n1 = DynamicParameter() {parameterIndex=2, defaultValue=<rgba 0.05 0.05 1 1>, paramNames [PanSpeedX, PanSpeedY, ChromaAmount, RedGreenAmount]}
  n2 = DynamicParameter() {parameterIndex=3, defaultValue=<rgba 256 1 1 1>, paramNames [MacroScale, MacroSpeed, DistortionSpeed, Param4]}
  n3 = Time()
  n4 = ScalarParameter() {group=Chromatic, parameterName=Panning Scale}
  n5 = Constant2Vector() {g=-0.028}
  n6 = DynamicParameter() {parameterIndex=3, defaultValue=<rgba 256 1 1 1>, paramNames [MacroScale, MacroSpeed, DistortionSpeed, Param4]}
  n7 = TextureSampleParameter2D(UVs=Panner(
    Coordinate=MaterialFunctionCall!(Plane Center, Plane Normal Axis, Scale=@n2.MacroScale) {materialFunction=/Game/Effects/Fort_Effects/Materials/_Functions/MF_MacroUV_Meshes_01}.Result
    Time=Multiply(A=@n3, B=@n4)
    Speed=Multiply(A=@n5, B=@n6.MacroSpeed)
   ) {speedY=-0.028}) {
   desc=Distort the Tiling Mask by this
   group=Chromatic
   parameterName=UVDistort_Normal_1
   samplerType=SAMPLERTYPE_Normal
   texture=/Game/Global/Textures/Default/Flowmaps/T_SwirlFlow_02_N
  }
  n8 = Multiply(A=ComponentMask(None=@n7.RGB) {g=true, r=true}, B=ScalarParameter() {defaultValue=0.08, group=Chromatic, parameterName=UVDistort})
  n9 = Add(A=@n8, B=Add(B=TextureCoordinate() {uTiling=0.5, vTiling=0.5}) {constA=0.1})
  n10 = DynamicParameter() {paramNames [TileX,TileY,OffsetX,OffsetY]}
  n11 = AppendVector(A=@n10.OffsetX, B=@n10.OffsetX)
  n12 = Multiply(A=Time(), B=ScalarParameter() {group=Chromatic, parameterName=Panning Scale})
  n13 = Multiply(A=Constant2Vector() {g=-0.0125}, B=ScalarParameter() {defaultValue=1, group=Chromatic, parameterName=Velocity_Ctrl})
  n14 = TextureSampleParameter2D(UVs=Panner(Coordinate=Add(A=@n9, B=@n11), Time=@n12, Speed=@n13) {speedY=-0.0125}) {
   desc=UV1 Tiling Noise Mask
   group=Chromatic
   parameterName=TileAlphaMask_2
   samplerType=SAMPLERTYPE_LinearGrayscale
   texture=/Game/Global/Textures/Default/Noise/T_SwirlNoise_N_height
  }
  n15 = Add(B=TextureCoordinate() {uTiling=0.5, vTiling=0.5}) {constA=-0.1}
  n16 = TextureSampleParameter2D(UVs=Panner(Coordinate=Add(A=@n8, B=@n15), Time=@n12, Speed=@n13) {speedY=-0.0125}) {
   desc=UV1 Tiling Noise Mask
   group=Chromatic
   parameterName=TileAlphaMask_1
   samplerType=SAMPLERTYPE_LinearGrayscale
   texture=&T_SwirlNoise_N_height
  }
  n17 = AppendVector(A=@n14.R, B=@n16.R)
  n18 = Add(A=@n8, B=TextureCoordinate() {uTiling=0.5, vTiling=0.5})
  n19 = TextureSampleParameter2D(UVs=Panner(Coordinate=Add(A=@n18, B=@n11), Time=@n12, Speed=@n13) {speedY=-0.0125}) {
   desc=UV1 Tiling Noise Mask
   group=Chromatic
   parameterName=TileAlphaMask_3
   samplerType=SAMPLERTYPE_LinearGrayscale
   texture=&T_SwirlNoise_N_height
  }
  n20 = DynamicParameter() {parameterIndex=2, defaultValue=<rgba 0.05 0.05 1 1>, paramNames [PanSpeedX, PanSpeedY, ChromaAmount, RedGreenAmount]}
  n21 = Desaturation(None=AppendVector(A=Multiply(A=@n1.RedGreenAmount, B=@n17), B=@n19.G), Fraction=@n20.ChromaAmount)
  MaterialFunctionCall(Hue Shift Percentage=ScalarParameter() {defaultValue=0.5}, Texture=@n21) {materialFunction=&HueShift}
  n22 = DynamicParameter() {paramNames [TileX,TileY,OffsetX,OffsetY]}
  n23 = AppendVector(A=@n22.OffsetX, B=@n22.OffsetX)
  n24 = TextureCoordinate()
  n25 = DynamicParameter() {paramNames [TileX,TileY,OffsetX,OffsetY]}
  n26 = AppendVector(A=@n25.TileX, B=@n25.TileX)
  n27 = DynamicParameter() {paramNames [TileX,TileY,OffsetX,OffsetY]}
  n28 = AppendVector(A=@n27.OffsetX, B=@n27.OffsetX)
  n29 = DynamicParameter() {parameterIndex=2, defaultValue=<rgba 0.05 0.05 1 1>, paramNames [PanSpeedX, PanSpeedY, ChromaAmount, RedGreenAmount]}
  n30 = TextureCoordinate()
  n31 = DynamicParameter() {parameterIndex=3, defaultValue=<rgba 256 1 1 1>, paramNames [MacroScale, MacroSpeed, DistortionSpeed, Param4]}
  n32 = Multiply(A=Time(), B=@n31.DistortionSpeed)
  n33 = TextureSampleParameter2D(UVs=Rotator(Coordinate=Multiply(A=@n30, B=@n26), Time=@n32) {centerX=1, speed=0.0312}) {
   parameterName=DistortionTexture
   samplerType=SAMPLERTYPE_LinearColor
   texture=/Game/Global/Textures/Default/Tileable_Patterns/T_Tile_Smoke_Small
  }
  n34 = TextureCoordinate()
  n35 = TextureSampleParameter2D(UVs=Rotator(Coordinate=Multiply(A=@n34, B=@n26), Time=Multiply(A=@n32) {constB=-1}) {centerY=1, speed=-0.0312}) {
   parameterName=DistortionTexture
   samplerType=SAMPLERTYPE_LinearColor
   texture=&T_Tile_Smoke_Small
  }
  n36 = DynamicParameter() {parameterIndex=1, defaultValue=<rgba 0.05 0.05 1 1>, paramNames [Contrast, DistortAmount, OffsetX, OffsetY]}
  n37 = Multiply(
   A=Add(A=ComponentMask(None=@n33.RGB) {b=true, g=true, r=true}, B=ComponentMask(None=@n35.RGB) {b=true, g=true, r=true})
   B=@n36.DistortAmount
  )
  n38 = Constant()
  n39 = ShadingPathSwitch(Default=QualitySwitch(Default=ComponentMask(None=@n37) {g=true, r=true}, Low=@n38), Mobile=@n38) {desc=Migrated from FeatureLevelSwitch}
  n40 = Add(A=Panner(Coordinate=Multiply(A=@n24, B=@n26), Speed=Multiply(A=@n28, B=@n29.PanSpeedX)) {speedX=0.02, speedY=0.15}, B=@n39)
  n41 = DynamicParameter() {paramNames [TileX,TileY,OffsetX,OffsetY]}
  n42 = AppendVector(A=@n41.OffsetX, B=@n41.OffsetX)
  n43 = Add(A=Panner(Coordinate=Multiply(A=@n24, B=@n26), Speed=Multiply(A=@n28, B=@n29.PanSpeedY)) {speedX=-0.02, speedY=0.1}, B=@n39)
  n44 = Add(
   A=TextureSampleParameter2D(UVs=Add(A=@n23, B=@n40)) {parameterName=SmokeTexture, texture=/Game/Global/Textures/Default/Caustics/T_Turbulence_Seamless}.G
   B=TextureSampleParameter2D(UVs=Add(A=@n42, B=@n43)) {parameterName=SmokeTexture, texture=&T_Turbulence_Seamless}.G
  )
  n45 = DynamicParameter() {parameterIndex=1, defaultValue=<rgba 0.05 0.05 1 1>, paramNames [Contrast, DistortAmount, OffsetX, OffsetY]}
  n46 = SphereMask(
   A=TextureCoordinate()
   B=Constant2Vector() {g=0.5, r=0.5}
   Radius=Multiply(A=ScalarParameter() {defaultValue=0.35, parameterName=EdgeSoften}, B=Power(Base=@n44, Exp=@n45.Contrast))
   Hardness=ScalarParameter() {parameterName=Hardness}
  )
  n47 = VectorParameter() {group=TOD, parameterName="Color_Morning (V3)", defaultValue=<rgba 1 1 1 0>}
  n48 = VectorParameter() {group=TOD, parameterName="Color_Day (V3)", defaultValue=<rgba 1 1 1 0>}
  n49 = VectorParameter() {group=TOD, parameterName="Color_Evening (V3)", defaultValue=<rgba 1 1 1 0>}
  n50 = VectorParameter() {group=TOD, parameterName="Color_Night (V3)", defaultValue=<rgba 1 1 1 0>}
  n51 = StaticSwitchParameter(
   True=Multiply(
    A=@n46
    B=MaterialFunctionCall(Color_Morning=@n47.RGB, Color_Day=@n48.RGB, Color_Evening=@n49.RGB, Color_Night=@n50.RGB) {materialFunction=&Color_Time_of_Day}.Result
   )
   False=@n46
  ) {parameterName=Use TOD?}
  n52 = Multiply(A=@n21, B=@n51)
  MaterialFunctionCall(Emissive=Multiply(A=StaticSwitchParameter(True=@n52, False=@n51) {parameterName=UseChromatic?}, B=ParticleColor().RGB)) {materialFunction=&MF_FX_EmissiveAdjustments}
  Multiply(A=TextureCoordinate(), B=ScalarParameter() {defaultValue=0.25, group=Chromatic, parameterName=TileSize_1})
  n53 = MaterialFunctionCall(
   Value_Morning=ScalarParameter() {defaultValue=1, group=Opacity, parameterName="Value_Morning (S)"}
   Value_Day=ScalarParameter() {defaultValue=1, group=Opacity, parameterName="Value_Day (S)"}
   Value_Evening=ScalarParameter() {defaultValue=1, group=Opacity, parameterName="Value_Evening (S)"}
   Value_Night=ScalarParameter() {defaultValue=1, group=Opacity, parameterName="Value_Night (S)"}
  ) {materialFunction=&Scalar_Time_of_Day}
  n54 = Multiply(A=StaticSwitchParameter(True=Multiply(A=@n46, B=@n53.Result), False=@n46) {parameterName=Use TOD?}, B=ParticleColor().A)
  n55 = DepthFade(Opacity=@n54, FadeDistance=ScalarParameter() {defaultValue=150, group=Opacity, parameterName=Opacity_Depth_FadeDistance})
  Multiply(B=@n55)
  Panner()
  QualitySwitch(Default=@n55, Low=@n54)
  ScalarParameter() {defaultValue=0.5, parameterName=Param_1}
  ScalarParameter() {defaultValue=256, parameterName=MacroScale}
  ScalarParameter() {defaultValue=1, parameterName=MacroSpeed}
 ]
}
