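# Unlit, two-sided translucent material for mesh particles that renders a soft shockwave tinted by particle color.
# Two smoke textures pan oppositely in stepped time, with dynamic parameters controlling speed and U tiling, then sum and erode.
# That mask blends darkened, optionally hue-shifted color toward particle color for emissive.
# It also drives opacity, with depth fade off mobile and an optional bottom gradient.
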Material /FNGameplayCues/Athena/Effects/Materials/Generic/M_Soft_Shockwave_Generic_ParticleColor
graph {
 properties {
  bEnableResponsiveAA=true
  bUsedWithMeshParticles=true
  blendMode=BLEND_Translucent
  pixelDepthOffsetMode=PDOM_Legacy
  shadingModel=MSM_Unlit
  twoSided=true
  usedShadingModels=Unlit
 }
 nodes [
  n1 = ParticleColor()
  n2 = StaticSwitchParameter(
   True=MaterialFunctionCall(Hue Shift Percentage=ScalarParameter() {defaultValue=0.15, parameterName=DarkColor_Hueshift}, Texture=@n1.RGB) {materialFunction=&HueShift}.Result
   False=@n1.RGB
  ) {defaultValue=true, parameterName=Use Hueshift?}
  n3 = ScalarParameter() {defaultValue=0.15, parameterName=DarkColor_Multiply}
  n4 = ComponentMask(None=VectorParameter() {parameterName=MaskOffset}.RGB) {g=true, r=true}
  n5 = TextureCoordinate() {uTiling=1.5, vTiling=0.5}
  n6 = AppendVector(
   A=DynamicParameter() {defaultValue=<rgba 17 1 10 0>, paramNames [Panner Speed, U Random, Electric Arc Multiplier, Erode]}.U Random
   B=Constant() {r=1}
  )
  n7 = Ceil(None=Multiply(
    A=Time()
    B=DynamicParameter() {defaultValue=<rgba 17 1 10 0>, paramNames [Panner Speed, U Random, Electric Arc Multiplier, Erode]}.Panner Speed
   ))
  n8 = Add(A=@n4, B=Panner(Coordinate=Multiply(A=@n5, B=@n6), Time=@n7) {bFractionalPart=true, speedX=1.11})
  n9 = TextureCoordinate() {uTiling=-0.88, vTiling=0.5}
  n10 = Add(A=Panner(Coordinate=Multiply(A=@n9, B=@n6), Time=@n7) {bFractionalPart=true, speedX=-0.83}, B=@n4)
  n11 = MaterialFunctionCall!(
   Dynamic Param=DynamicParameter() {defaultValue=<rgba 17 1 10 0>, paramNames [Panner Speed, U Random, Electric Arc Multiplier, Erode]}.Erode
   Alpha=Add(
    A=TextureSample(UVs=@n8) {texture=/Game/Effects/Fort_Effects/Textures/Smoke/T_Smoke_Trail_Soft_2}.R
    B=TextureSample(UVs=@n10) {texture=&T_Smoke_Trail_Soft_2}.R
   )
  ) {materialFunction=/Game/Effects/Fort_Effects/Materials/_Functions/Erode_Alpha}
  n12 = Clamp(None=@n11.Result)
  n13 = ShadingPathSwitch(Default=DepthFade(FadeDistance=ScalarParameter() {defaultValue=50, parameterName=Depth Fade Distance}), Mobile=Constant() {r=1}) {desc=Migrated from FeatureLevelSwitch}
  MaterialFunctionCall(Emissive=LinearInterpolate(A=Multiply(A=@n2, B=@n3), B=@n1.RGB, Alpha=Multiply(A=@n12, B=@n13)) {constAlpha=0}) {materialFunction=&MF_FX_EmissiveAdjustments}
  n14 = Multiply(A=@n11.Result, B=@n13)
  n15 = MaterialFunctionCall() {materialFunction=&LinearGradient}
  n16 = ScalarParameter() {defaultValue=1, parameterName=BottomFalloff}
  n17 = Multiply(A=Power(Base=OneMinus(None=@n15.VGradient), Exp=@n16), B=@n16)
  StaticSwitchParameter(True=Multiply(A=@n14, B=Clamp(None=@n17)), False=@n14) {parameterName=UseBottomGradient}
 ]
}
